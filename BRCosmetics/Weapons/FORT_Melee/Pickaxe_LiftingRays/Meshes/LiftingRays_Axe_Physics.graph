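# Physics asset for the LiftingRays pickaxe cosmetic, defining bodies for its eight-segment canopy.
# A base capsule anchors eight radial two-link capsule chains, each running from the base through an A segment to a B segment.
# Every joint has zero twist and swing1 limits, leaving a single hinge-like swing2 axis, with a looser limit at the outer A-to-B joints than at the base.
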PhysicsAsset /BRCosmetics/Weapons/FORT_Melee/Pickaxe_LiftingRays/Meshes/LiftingRays_Axe_Physics
bodies {
 dyn_BASE_canopy [{shapeType=Capsule, length=7.4204, radius=2.551, shapeName=dyn_BASE_canopy_capsule}]
 dyn_canopy_A_0 [{shapeType=Capsule, length=5, radius=1.5, shapeName=dyn_canopy_A_0_capsule, center=<2 -10.3886 -2.204>, rotation=<rot -80 -90 180>}]
 dyn_canopy_A_1 [{shapeType=Capsule, length=5, radius=1.5, shapeName=dyn_canopy_A_1_capsule, center=<2 -10.3886 -2.204>, rotation=<rot -80 -90 180>}]
 dyn_canopy_A_2 [{shapeType=Capsule, length=5, radius=1.5, shapeName=dyn_canopy_A_2_capsule, center=<2 -10.3886 -2.204>, rotation=<rot -80 -90 180>}]
 dyn_canopy_A_3 [{shapeType=Capsule, length=5, radius=1.5, shapeName=dyn_canopy_A_3_capsule, center=<2 -10.3886 -2.204>, rotation=<rot -80 -90 180>}]
 dyn_canopy_A_4 [{shapeType=Capsule, length=5, radius=1.5, shapeName=dyn_canopy_A_4_capsule, center=<2 -10.3886 -2.204>, rotation=<rot -80 -90 180>}]
 dyn_canopy_A_5 [{shapeType=Capsule, length=5, radius=1.5, shapeName=dyn_canopy_A_5_capsule, center=<2 -10.3886 -2.204>, rotation=<rot -80 -90 180>}]
 dyn_canopy_A_6 [{shapeType=Capsule, length=5, radius=1.5, shapeName=dyn_canopy_A_6_capsule, center=<2 -10.3886 -2.204>, rotation=<rot -80 -90 180>}]
 dyn_canopy_A_7 [{shapeType=Capsule, length=5, radius=1.5, shapeName=dyn_canopy_A_7_capsule, center=<2 -10.3886 -2.204>, rotation=<rot -80 -90 180>}]
 dyn_canopy_B_0 [{shapeType=Capsule, length=5, radius=2, shapeName=dyn_canopy_B_0_capsule, center=<-5 -3.2001 -2.0786>, rotation=<rot -60 -90 180>}]
 dyn_canopy_B_1 [{shapeType=Capsule, length=5, radius=2, shapeName=dyn_canopy_B_1_capsule, center=<-5 -3.2001 -2.0786>, rotation=<rot -60 -90 180>}]
 dyn_canopy_B_2 [{shapeType=Capsule, length=5, radius=2, shapeName=dyn_canopy_B_2_capsule, center=<-5 -3.2001 -2.0786>, rotation=<rot -60 -90 180>}]
 dyn_canopy_B_3 [{shapeType=Capsule, length=5, radius=2, shapeName=dyn_canopy_B_3_capsule, center=<-5 -3.2001 -2.0786>, rotation=<rot -60 -90 180>}]
 dyn_canopy_B_4 [{shapeType=Capsule, length=5, radius=2, shapeName=dyn_canopy_B_4_capsule, center=<-5 -3.2001 -2.0786>, rotation=<rot -60 -90 180>}]
 dyn_canopy_B_5 [{shapeType=Capsule, length=5, radius=2, shapeName=dyn_canopy_B_5_capsule, center=<-5 -3.2001 -2.0786>, rotation=<rot -60 -90 180>}]
 dyn_canopy_B_6 [{shapeType=Capsule, length=5, radius=2, shapeName=dyn_canopy_B_6_capsule, center=<-5 -3.2001 -2.0786>, rotation=<rot -60 -90 180>}]
 dyn_canopy_B_7 [{shapeType=Capsule, length=5, radius=2, shapeName=dyn_canopy_B_7_capsule, center=<-5 -3.2001 -2.0786>, rotation=<rot -60 -90 180>}]
}
constraints [
 {bone1Name=dyn_canopy_B_0, bone2Name=dyn_canopy_A_0, swing1LimitDegrees=0, swing2LimitDegrees=30, twistLimitDegrees=0}
 {bone1Name=dyn_canopy_B_1, bone2Name=dyn_canopy_A_1, swing1LimitDegrees=0, swing2LimitDegrees=30, twistLimitDegrees=0}
 {bone1Name=dyn_canopy_B_2, bone2Name=dyn_canopy_A_2, swing1LimitDegrees=0, swing2LimitDegrees=30, twistLimitDegrees=0}
 {bone1Name=dyn_canopy_B_3, bone2Name=dyn_canopy_A_3, swing1LimitDegrees=0, swing2LimitDegrees=30, twistLimitDegrees=0}
 {bone1Name=dyn_canopy_B_4, bone2Name=dyn_canopy_A_4, swing1LimitDegrees=0, swing2LimitDegrees=30, twistLimitDegrees=0}
 {bone1Name=dyn_canopy_B_5, bone2Name=dyn_canopy_A_5, swing1LimitDegrees=0, swing2LimitDegrees=30, twistLimitDegrees=0}
 {bone1Name=dyn_canopy_B_6, bone2Name=dyn_canopy_A_6, swing1LimitDegrees=0, swing2LimitDegrees=30, twistLimitDegrees=0}
 {bone1Name=dyn_canopy_B_7, bone2Name=dyn_canopy_A_7, swing1LimitDegrees=0, swing2LimitDegrees=30, twistLimitDegrees=0}
 {bone1Name=dyn_canopy_A_7, bone2Name=dyn_BASE_canopy, swing1LimitDegrees=0, swing2LimitDegrees=20, twistLimitDegrees=0}
 {bone1Name=dyn_canopy_A_6, bone2Name=dyn_BASE_canopy, swing1LimitDegrees=0, swing2LimitDegrees=20, twistLimitDegrees=0}
 {bone1Name=dyn_canopy_A_5, bone2Name=dyn_BASE_canopy, swing1LimitDegrees=0, swing2LimitDegrees=20, twistLimitDegrees=0}
 {bone1Name=dyn_canopy_A_4, bone2Name=dyn_BASE_canopy, swing1LimitDegrees=0, swing2LimitDegrees=20, twistLimitDegrees=0}
 {bone1Name=dyn_canopy_A_3, bone2Name=dyn_BASE_canopy, swing1LimitDegrees=0, swing2LimitDegrees=20, twistLimitDegrees=0}
 {bone1Name=dyn_canopy_A_2, bone2Name=dyn_BASE_canopy, swing1LimitDegrees=0, swing2LimitDegrees=20, twistLimitDegrees=0}
 {bone1Name=dyn_canopy_A_1, bone2Name=dyn_BASE_canopy, swing1LimitDegrees=0, swing2LimitDegrees=20, twistLimitDegrees=0}
 {bone1Name=dyn_canopy_A_0, bone2Name=dyn_BASE_canopy, swing1LimitDegrees=0, swing2LimitDegrees=20, twistLimitDegrees=0}
]
